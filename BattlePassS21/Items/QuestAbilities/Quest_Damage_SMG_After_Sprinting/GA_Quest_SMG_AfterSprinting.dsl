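; Quest gameplay ability for the Season 21 Battle Pass "damage with SMG after sprinting" challenge, granting a short post-sprint effect on the player.
; On event activation it calls the parent, then waits for the TacticalSprint gameplay tag to be removed from the owner.
; When the tag is removed, it applies the AfterSprinting5s gameplay effect to the owning pawn's Fort ability system component and ends itself.
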(event Ability|EventActivateAbilityFromEvent
  (params (EventData :type Gameplay%20Event%20Data%20Structure%20%28by%20ref%29))
  (|CallParentFunction EventData)
  (Ability|Tasks|WaitGameplayTagRemove "(TagName=\"Gameplay.Action.Player.TacticalSprint\")" 0 true
    (:Removed
      (GameplayEffects|ApplyGameplayEffectToSelf :self (Ability|GetFortAbilitySystemComponent (Variables|GAVars|Data|GetOwningPlayerPawn)) :GameplayEffectClass (Variables|Default|GetGE_Quest_AfterSprinting5s) :Level 0.0)
      (Ability|EndAbility))))
